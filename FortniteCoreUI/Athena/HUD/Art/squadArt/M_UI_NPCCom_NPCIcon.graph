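# Translucent UI material for an NPC icon in the squad HUD art, drawing a swappable icon texture clipped to a circle.
# A circle function with adjustable size multiplies both the texture's color and its alpha.
# The masked alpha blends toward a copy multiplied by a configurable, rotatable linear gradient, giving an optional fade.
# The gradient intensity parameter controls that blend and defaults to off.
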Material /FortniteCoreUI/Athena/HUD/Art/squadArt/M_UI_NPCCom_NPCIcon
graph {
 properties {
  bUsedWithStaticMesh=false
  blendMode=BLEND_Translucent
  materialDomain=MD_UI
  pixelDepthOffsetMode=PDOM_Legacy
  usedShadingModels=Default Lit
 }
 nodes [
  n1 = TextureSampleParameter2D() {parameterName=NPC_Icon, texture=/FortniteCoreUI/Athena/HUD/Interaction/T_UI_NPC_Comm_Default}
  n2 = MaterialFunctionCall(Size=ScalarParameter() {defaultValue=1, parameterName=Circle_Size}) {materialFunction=&MF_UI_Circle}
  n3 = Multiply(A=@n1.A, B=@n2.Fill)
  n4 = ScalarParameter() {defaultValue=-1.5, group=Gradient Settings, parameterName=GradientStart, sortPriority=1}
  n5 = ScalarParameter() {group=Gradient Settings, parameterName=GradientEnd, sortPriority=2}
  n6 = ScalarParameter() {defaultValue=0.25, group=Gradient Settings, parameterName=GradientRotation, sortPriority=4}
  n7 = ScalarParameter() {defaultValue=-0.075, group=Gradient Settings, parameterName=GradientOffset, sortPriority=3}
  n8 = VectorParameter() {
   group=Gradient Settings
   parameterName=GradientRotationCenter
   sortPriority=5
   channelNames {a="", b="", g=Center Y, r=Center X}
   defaultValue=<rgba 0.5 0.5 0 0>
  }
  LinearInterpolate(
   A=@n3
   B=Multiply(
    A=@n3
    B=MaterialFunctionCall(GradientStart=@n4, GradientEnd=@n5, RotationAngle=@n6, RotationCenter=@n7, GradientOffset=@n8) {materialFunction=&MF_UI_LinearGradient}.Result
   )
   Alpha=ScalarParameter() {group=Gradient Settings, parameterName=GradientIntensity}
  )
  Multiply(A=@n1.RGB, B=@n2.Fill)
 ]
}
